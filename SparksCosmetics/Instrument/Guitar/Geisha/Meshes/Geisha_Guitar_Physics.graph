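# Physics asset for the Geisha guitar instrument cosmetic, defining the bodies and joints for its dangling cloth and rope bones.
# Each bone gets a capsule body, with a long one on the handle serving as the common root.
# Four cloth bodies attach separately to the handle with asymmetric swing limits, while three rope bodies form a chain from it.
# All joints lock twist, and rope swing limits widen toward the tip.
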PhysicsAsset /SparksCosmetics/Instrument/Guitar/Geisha/Meshes/Geisha_Guitar_Physics
bodies {
 dyn_cloth_01 [{shapeType=Capsule, length=2, radius=2, shapeName=dyn_cloth_01_capsule, rotation=<rot 90 0 0>}]
 dyn_cloth_02 [{shapeType=Capsule, length=2, radius=2, shapeName=dyn_cloth_02_capsule, rotation=<rot 90 0 0>}]
 dyn_cloth_03 [{shapeType=Capsule, length=2, radius=2, shapeName=dyn_cloth_03_capsule, rotation=<rot 90 0 0>}]
 dyn_cloth_04 [{shapeType=Capsule, length=2, radius=2, shapeName=dyn_cloth_04_capsule, rotation=<rot 90 0 0>}]
 dyn_rope_01 [{shapeType=Capsule, length=0.7164, radius=2, shapeName=dyn_rope_01_capsule, center=<0.5 0 0>, rotation=<rot 90 0 0>}]
 dyn_rope_02 [{shapeType=Capsule, length=2.6932, radius=2, shapeName=dyn_rope_02_capsule, center=<0.6 0 0>, rotation=<rot 90 0 0>}]
 dyn_rope_03 [{shapeType=Capsule, length=3.4173, radius=2.3772, shapeName=dyn_rope_03_capsule, center=<3 0 0>, rotation=<rot 84.1764 -180 -180>}]
 handle [{shapeType=Capsule, length=50, radius=2, center=<0.3059 -0.0837 27.4668>, rotation=<rot -0.7118 -90 -0.2255>}]
}
constraints [
 {bone1Name=dyn_cloth_01, bone2Name=handle, swing1LimitDegrees=30, swing2LimitDegrees=5, twistLimitDegrees=45, twistMotion=Locked}
 {bone1Name=dyn_cloth_02, bone2Name=handle, swing1LimitDegrees=30, swing2LimitDegrees=5, twistLimitDegrees=45, twistMotion=Locked}
 {bone1Name=dyn_cloth_03, bone2Name=handle, swing1LimitDegrees=30, swing2LimitDegrees=5, twistLimitDegrees=45, twistMotion=Locked}
 {bone1Name=dyn_cloth_04, bone2Name=handle, swing1LimitDegrees=30, swing2LimitDegrees=5, twistLimitDegrees=45, twistMotion=Locked}
 {bone1Name=dyn_rope_01, bone2Name=handle, swing1LimitDegrees=45, swing2LimitDegrees=45, twistLimitDegrees=45, twistMotion=Locked}
 {bone1Name=dyn_rope_02, bone2Name=dyn_rope_01, swing1LimitDegrees=50, swing2LimitDegrees=50, twistLimitDegrees=45, twistMotion=Locked}
 {bone1Name=dyn_rope_03, bone2Name=dyn_rope_02, swing1LimitDegrees=60, swing2LimitDegrees=60, twistLimitDegrees=45, twistMotion=Locked}
]
